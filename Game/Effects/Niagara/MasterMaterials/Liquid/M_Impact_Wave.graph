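# Translucent, two-sided unlit Niagara material for ribbons and mesh particles, apparently for liquid impact wave effects.
# A slowly panning caustics noise distorts UVs of a faster-panning plasma texture, whose powered result is tinted by particle color, optionally desaturated, and fed through an emissive adjustment function.
# An opacity-style chain multiplies that pattern by a soft trail texture, particle alpha, view-angle fade and a V gradient.
# The ColorA/ColorB blend feeds no other node.
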Material /Game/Effects/Niagara/MasterMaterials/Liquid/M_Impact_Wave
graph {
 properties {
  bUsedWithNiagaraMeshParticles=true
  bUsedWithNiagaraRibbons=true
  blendMode=BLEND_Translucent
  pixelDepthOffsetMode=PDOM_Legacy
  shadingModel=MSM_Unlit
  twoSided=true
  usedShadingModels=Unlit
 }
 nodes [
  n1 = TextureCoordinate()
  n2 = TextureSampleParameter2D(UVs=Panner(Coordinate=TextureCoordinate() {uTiling=0.5, vTiling=0.5}) {speedX=0.2}) {parameterName=TexA, samplerType=SAMPLERTYPE_LinearGrayscale, texture=/Game/Global/Textures/Default/Caustics/T_Nosie_Caustics_Grayscale}
  n3 = Multiply(A=@n2.R) {constB=0.2}
  n4 = TextureSampleParameter2D(UVs=Panner(Coordinate=Add(A=@n1, B=@n3)) {speedX=-1}) {
   parameterName=TexB
   samplerType=SAMPLERTYPE_LinearGrayscale
   texture=/Game/Effects/Fort_Effects/Emotes/UltraLight/Texture/T_Plasma_D_Greyscale
  }
  n5 = Power(Base=@n4.R)
  LinearInterpolate(
   A=VectorParameter() {parameterName=ColorB, defaultValue=<rgba 0.3239 0.3883 2 1>}.RGB
   B=VectorParameter() {parameterName=ColorA, defaultValue=<rgba 0.7224 0.3316 2 1>}.RGB
   Alpha=@n5
  )
  n6 = ParticleColor()
  MaterialFunctionCall(Emissive=Desaturation(None=Multiply(A=@n6.RGB, B=@n5), Fraction=ScalarParameter() {parameterName=DesatFraction})) {materialFunction=&MF_FX_EmissiveAdjustments}
  n7 = TextureCoordinate()
  n8 = TextureSampleParameter2D() {parameterName=TexRibbon, samplerType=SAMPLERTYPE_Grayscale, texture=&T_Smoke_Trail_Soft}
  n9 = Multiply(A=@n5, B=Multiply(A=@n8.R))
  n10 = Constant() {r=1}
  Multiply(
   A=Power(Base=ComponentMask(None=@n7) {g=true})
   B=Multiply(
    A=Multiply(A=@n6.A, B=@n9)
    B=MaterialFunctionCall("FadeSensitivity (optional)"=@n10) {materialFunction=&FadeBasedOnViewAngle}.Result
   )
  )
 ]
}
