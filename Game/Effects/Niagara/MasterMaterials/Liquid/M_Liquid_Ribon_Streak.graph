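# Translucent, subsurface-shaded master material for Niagara ribbons that draws lit liquid streaks.
# Ribbon UVs are stretched or tiled, narrowed by an S-curve and sine wobble, then distorted by a panning flowmap texture.
# These UVs drive a sphere normal map and a sphere mask; particle erosion shrinks the mask and exaggerates the normals.
# Dynamic parameters supply erode, wobble, tile, separation and pan, while Fresnel and particle color build emissive.
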Material /Game/Effects/Niagara/MasterMaterials/Liquid/M_Liquid_Ribon_Streak
graph {
 properties {
  bTangentSpaceNormal=false
  bUsedWithNiagaraRibbons=true
  blendMode=BLEND_Translucent
  pixelDepthOffsetMode=PDOM_Legacy
  shadingModel=MSM_Subsurface
  translucencyLightingMode=TLM_Surface
  usedShadingModels=Subsurface
 }
 nodes [
  n1 = OneMinus(None=Fresnel(ExponentIn=ScalarParameter() {defaultValue=0.5, parameterName=EmissiveFresnelExp}) {baseReflectFraction=0, exponent=7})
  MaterialFunctionCall(Emissive=Multiply(A=ParticleColor().RGB, B=Power(Base=@n1) {constExponent=3})) {materialFunction=&MF_FX_EmissiveAdjustments}
  MaterialFunctionCall() {materialFunction=&DebugTimeSine}
  n2 = MaterialFunctionCall(UV Channel=TextureCoordinate()) {materialFunction=&LinearGradient}
  n3 = NamedRerouteUsage()
  n4 = AppendVector(
   A=@n2.VGradient
   B=MaterialFunctionCall!(Center Point=Constant() {r=0.5}, 0-1 Gradient=@n2.UGradient, Scale Factor=@n3) {materialFunction=/Engine/Functions/Engine_MaterialFunctions02/Gradients/StretchGradient}.Resulting Gradient
  )
  n5 = AppendVector(A=@n2.VGradient, B=Multiply(A=@n2.UGradient, B=@n3))
  n6 = NamedRerouteUsage()
  n7 = OneMinus(None=@n2.UGradient)
  n8 = NamedRerouteUsage()
  n9 = Multiply(A=@n2.UGradient, B=Multiply(A=NamedRerouteUsage()) {constB=0.5})
  n10 = Add(
   A=Multiply(A=Time(), B=ScalarParameter() {defaultValue=-0.2, parameterName=PanSpeed})
   B=DynamicParameter() {parameterIndex=1, paramNames [Pan,WidthMult,Param3,Param4]}.Pan
  )
  n11 = ConstantBiasScale(None=Sine(None=Add(A=@n9, B=@n10)))
  n12 = ScalarParameter() {defaultValue=-0.2, parameterName=SineWobbleSizeSubtract}
  n13 = Add(
   A=MaterialFunctionCall(In=@n7, Power=LinearInterpolate(Alpha=@n8) {constA=0.5, constB=0}) {materialFunction=&SCurve}.Result
   B=Multiply(A=Multiply(A=@n11, B=@n12) {constB=0.1}, B=NamedRerouteUsage())
  )
  n14 = Constant() {r=1}
  n15 = NamedRerouteUsage()
  n16 = TextureCoordinate() {uTiling=0.5}
  n17 = AppendVector(A=NamedRerouteUsage(), B=Constant() {r=1})
  n18 = TextureSample(UVs=Panner(Coordinate=Multiply(A=@n16, B=@n17), Time=@n10) {speedX=-1, speedY=0.1}) {samplerType=SAMPLERTYPE_LinearColor, texture=/Game/Environments/Apollo/Sets/Jerky/Materials/T_JerkyFlowmap}
  n19 = Multiply(A=ConstantBiasScale(None=ComponentMask(None=@n18.RGB) {g=true, r=true}) {bias=-0.5, scale=2}) {constB=0.2}
  n20 = Add(
   A=MaterialFunctionCall(UVs=LinearInterpolate(A=@n4, B=@n5, Alpha=@n6), Texture Scale=AppendVector(A=@n13, B=@n14)) {materialFunction=&ScaleUVsByCenter}.UVs
   B=LinearInterpolate(A=Multiply(A=@n15, B=@n19), Alpha=@n7) {constB=0}
  )
  n21 = ScalarParameter() {defaultValue=5, parameterName=EndingNormalIntesnity}
  n22 = Constant() {r=1}
  n23 = DynamicParameter() {defaultValue=<rgba 0 0 7 0>, paramNames [Erode,Wobble,Tile,Separate]}
  n24 = Power(Base=NamedRerouteDeclaration(None=@n23.Erode) {name=ErodeParam, nodeColor=<rgba 0.7344 0 1 1>}) {constExponent=3}
  Multiply(
   A=TextureSample(UVs=@n20) {samplerType=SAMPLERTYPE_Normal, texture=&T_Sphere_N}.RGB
   B=LinearInterpolate(
    B=MaterialFunctionCall(R=@n21, G=@n21, B=@n22) {materialFunction=&AppendMany}.RGB
    Alpha=@n24
   ) {constA=1}
  )
  NamedRerouteDeclaration(None=@n23.Tile) {name=TileParam, nodeColor=<rgba 0.5469 0 1 1>}
  NamedRerouteDeclaration(None=@n23.Wobble) {name=WobbleParam, nodeColor=<rgba 0.7344 0 1 1>}
  NamedRerouteDeclaration(None=@n23.Separate) {name=SeparateParam, nodeColor=<rgba 0.9922 0 1 1>}
  n25 = MaterialFunctionCall(Float2=@n20) {materialFunction=&BreakOutFloat2Components}
  n26 = Fmod(A=@n25.G, B=Constant() {r=2})
  Reroute(None=SphereMask(A=AppendVector(A=@n26, B=@n25.R), B=Constant() {r=0.5}, Radius=LinearInterpolate(Alpha=@n24) {constA=0.5, constB=0}))
  ScalarParameter() {parameterName=Metallic}
  ScalarParameter() {defaultValue=1, parameterName=Specular}
  ScalarParameter() {parameterName=Roughness}
  VectorParameter() {parameterName=Base Color, defaultValue=<rgba 0.5508 0.0056 0.6875 1>}
  VectorParameter() {parameterName=SubSurface Color, defaultValue=<rgba 0.4375 0.0246 0.3018 1>}
 ]
}
