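# Material function for Slurp liquid FX that sets base color, emissive, specular, roughness, metallic and world position offset.
# A per-vertex spherical harmonics decode of vertex-color coefficients along the skinning-corrected camera direction feeds the inner-to-outer color blend.
# Sun-driven scattering, input AO and time-of-day scaling shade the color, which also feeds emissive.
# Masked, time-animated world and local sine waves push vertices along their normals for wobble.
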MaterialFunction /Game/Effects/Niagara/MasterMaterials/Liquid/MF_Slurp_forFX
graph {nodes [
  Constant3Vector()
  n1 = TextureSampleParameter2D() {group=Slurp, parameterName=Slurp_WPO_Mask, samplerType=SAMPLERTYPE_Masks, texture=&T_WhiteMask}
  n2 = WorldPosition() {worldPositionShaderOffset=WPT_ExcludeAllShaderOffsets}
  n3 = Subtract(A=Frac(None=Multiply(A=@n2.XYZ) {constB=0.0025})) {constB=0.5}
  n4 = MaterialFunctionCall(Float3=Abs(None=Multiply(A=@n3) {constB=2})) {materialFunction=&BreakOutFloat3Components}
  n5 = Add(A=Add(A=@n4.R, B=@n4.R), B=@n4.B)
  n6 = Multiply(A=Time()) {constB=0.25}
  n7 = Sine(None=Subtract(A=Subtract(A=@n5) {constB=2}, B=Frac(None=@n6)))
  n8 = Multiply(A=Multiply(A=Add(A=@n7)) {constB=0.5}, B=ScalarParameter() {defaultValue=1, group=Slurp, parameterName=Slurp_WPO_WorldNoiseScale})
  n9 = MaterialFunctionCall(Float3=Multiply(A=NamedRerouteUsage(), B=ScalarParameter() {defaultValue=0.02, group=Slurp, parameterName=Slurp_WPO_LocalNoise_Size}) {constB=0.02}) {materialFunction=&BreakOutFloat3Components}
  n10 = Add(A=@n9.R, B=@n9.R)
  n11 = Subtract(A=Time(), B=ScalarParameter() {group=Slurp, parameterName=Slurp_WPO_random_seed})
  n12 = ScalarParameter() {defaultValue=0.6, group=Slurp, parameterName=Slurp_WPO_LocalNoise_Speed}
  n13 = Subtract(A=Subtract(A=Add(A=@n10, B=@n9.B)) {constB=2}, B=Frac(None=Multiply(A=@n11, B=@n12) {constB=0.6}))
  n14 = Multiply(A=Add(A=Sine(None=@n13))) {constB=0.5}
  n15 = ComponentMask(None=MaterialFunctionCall() {materialFunction=/Engine/Functions/Engine_MaterialFunctions02/Texturing/3DSandMayaUVCoordinates}.UVChannel1) {r=true}
  n16 = OneMinus(None=ScalarParameter() {defaultValue=1, parameterName=Slurp_WPO_UVMask})
  n17 = Multiply(A=Multiply(A=VertexNormalWS(), B=Add(A=@n8, B=@n14)), B=Saturate(None=Add(A=@n15, B=@n16)))
  n18 = Constant() {r=1}
  n19 = FunctionInput() {bUsePreviewValueAsDefault=true, description=AmbientOcclusion, inputName=AO, inputType=FunctionInput_Scalar, previewValue=<1 1 1 1>}
  n20 = ScalarParameter() {group=Slurp, parameterName=Slurp_AO_Intensity}
  n21 = ScalarParameter() {defaultValue=0.66, group=Slurp, parameterName=Slurp_TimeofDayScale_Morning}
  n22 = ScalarParameter() {defaultValue=1, group=Slurp, parameterName=Slurp_TimeofDayScale_Day}
  n23 = ScalarParameter() {defaultValue=0.66, group=Slurp, parameterName=Slurp_TimeofDayScale_Evening}
  n24 = ScalarParameter() {defaultValue=0.5, group=Slurp, parameterName=Slurp_TimeofDayScale_Night}
  n25 = LinearInterpolate(
   A=Constant3Vector() {constant=<rgba 1 1 1 0>}
   B=VectorParameter() {group=Slurp, parameterName=Slurp_InnerBackScatteringMultiplier, defaultValue=<rgba 0 0 0 1>}.RGB
   Alpha=NamedRerouteUsage()
  )
  n26 = ComponentMask(None=NamedRerouteUsage()) {b=true}
  n27 = ScalarParameter() {defaultValue=125, group=Slurp, parameterName=Slurp_OuterColor_TransitionStart}
  n28 = Saturate(None=Multiply(
    A=Subtract(A=@n26, B=@n27) {constB=125}
    B=ScalarParameter() {defaultValue=0.015, group=Slurp, parameterName=Slurp_OuterColor_TransitionScale}
   ) {constB=0.015})
  n29 = VectorParameter() {group=Slurp, parameterName=Slurp_OuterColorB, defaultValue=<rgba 0 0.5903 0.75 0>}
  n30 = VectorParameter() {group=Slurp, parameterName=Slurp_OuterColorA, defaultValue=<rgba 0.0125 0.724 0.6663 1>}
  n31 = MaterialFunctionCall() {materialFunction=&3DSandMayaUVCoordinates}
  n32 = OneMinus(None=Saturate(None=ComponentMask(None=@n31.UVChannel1) {r=true}))
  n33 = ScalarParameter() {group=Slurp, parameterName=Slurp_UV1_Darkening}
  n34 = NamedRerouteUsage()
  n35 = QualitySwitch(Default=Add(A=NamedRerouteUsage(), B=Multiply(A=@n32, B=@n33) {constB=0.0333}), Low=@n34)
  n36 = TextureSampleParameter2D() {group=Slurp, parameterName=Slurp_SHBiasMap, samplerType=SAMPLERTYPE_Masks, texture=/Game/Global/Textures/Default/Blanks/T_MidgreyMask}
  n37 = Subtract(A=@n36.G) {constB=0.5}
  n38 = Add(
   A=ShadingPathSwitch(Default=@n35, Deferred=@n35, Forward=@n34, Mobile=@n34)
   B=Multiply(A=Multiply(A=@n37) {constB=2}, B=ScalarParameter() {defaultValue=1, group=Slurp, parameterName=Slurp_SHBias_ScaleFactor})
  )
  n39 = LinearInterpolate(
   A=LinearInterpolate(
    A=VectorParameter() {group=Slurp, parameterName=Slurp_InnerColor, defaultValue=<rgba 0 0.0892 0.25 1>}.RGB
    B=LinearInterpolate(A=@n30.RGB, B=@n29.RGB, Alpha=@n28)
    Alpha=Saturate(None=@n38)
   )
   B=VectorParameter() {group=Slurp, parameterName=Slurp_InnerScatteringColor, defaultValue=<rgba 0.0877 0.0586 0.2031 1>}.RGB
   Alpha=NamedRerouteUsage()
  )
  n40 = Multiply(
   A=Power(Base=LinearInterpolate(A=@n18, B=@n19, Alpha=@n20), Exp=ScalarParameter() {defaultValue=1, group=Slurp, parameterName=Slurp_AO_pow})
   B=Multiply(
    A=MaterialFunctionCall(Value_Morning=@n21, Value_Day=@n22, Value_Evening=@n23, Value_Night=@n24) {materialFunction=&Scalar_Time_of_Day}.Result
    B=Multiply(A=@n25, B=@n39)
   )
  )
  n41 = ScalarParameter() {defaultValue=0.5, group=Slurp, parameterName=Slurp_BaseColor_Emissive_Intensity, sliderMax=1}
  n42 = SetMaterialAttributes!(
   MaterialAttributes
   Specular=ScalarParameter() {defaultValue=0.5, group=Slurp, parameterName=Slurp_Specular}
   Roughness=ScalarParameter() {defaultValue=0.2, group=Slurp, parameterName=Slurp_Roughness}
   Metallic=ScalarParameter() {group=Slurp, parameterName=Slurp_Metalness}
   BaseColor=@n40
   EmissiveColor=Multiply(A=@n40, B=Saturate(None=@n41))
   WorldPositionOffset=Multiply(A=Multiply(A=@n1.R, B=@n17), B=ScalarParameter() {defaultValue=1, group=Slurp, parameterName=Slurp_WPO_FinalScale})
  ) {attributeSetTypes [
    #1
    #2
    #3
    #4
    #5
    #6
   ]}
  FunctionOutput(None=@n42)
  MaterialFunctionCall() {materialFunction=&CameraDirectionVector}
  n43 = MaterialFunctionCall(
   BaseNormal=PreSkinnedNormal()
   AdditionalNormal=Transform(None=FunctionInput() {inputName=World Normal, previewValue=<0 0 1 1>}) {transformSourceType=TRANSFORMSOURCE_World, transformType=TRANSFORM_Local}
  ) {materialFunction=&BlendAngleCorrectedNormals}
  NamedRerouteDeclaration(None=ShaderStageSwitch(PixelShader=VertexInterpolator(VS=@n43.Result).PS, VertexShader=@n43.Result)) {name=PreSkinnedNrm_VS-PS, nodeColor=<rgba 0 0.4844 1 1>}
  n44 = PreSkinnedPosition()
  n45 = VertexInterpolator(VS=@n44)
  NamedRerouteDeclaration(None=FunctionInput(Preview=ShaderStageSwitch(PixelShader=@n45.PS, VertexShader=@n44)) {bUsePreviewValueAsDefault=true, description=Defaults to Particle Position, inputName=Pre Skinned Position}) {name=PreSkinnedPos_VS-PS, nodeColor=<rgba 0 0.4844 1 1>}
  n46 = VectorParameter() {group=Slurp, parameterName=SHMax, defaultValue=<rgba 1.25 0.13 0.18 0.13>}
  n47 = VectorParameter() {group=Slurp, parameterName=SHMin, defaultValue=<rgba 1 -0.16 -0.17 -0.15>}
  n48 = AppendVector(A=@n47.RGB, B=@n47.RGB)
  n49 = Subtract(A=AppendVector(A=@n46.RGB, B=@n46.RGB), B=@n48)
  n50 = VertexColor()
  n51 = AppendVector(A=@n50, B=@n50)
  n52 = Multiply(A=MaterialFunctionCall(X=@n50.G, Y=@n50.G, Z=@n50.G) {materialFunction=&MakeFloat3}.Result)
  MaterialExpressionStaticSwitchParameter_32 = StaticSwitchParameter(True=Add(A=Multiply(A=@n49, B=@n51), B=@n48), False=AppendVector(A=Add(A=@n50.R), B=Add(A=@n52) {constB=-0.5})) {group=Slurp, parameterName=UseCustomSH Fit}
  n53 = Multiply(
   A=MaterialFunctionCall() {materialFunction=&CameraDirectionVector}.Result
   B=ScalarParameter() {defaultValue=1, group=Slurp, parameterName=Slurp_DirectionVectorFlip}
  )
  MaterialExpressionMaterialFunctionCall_133 = MaterialFunctionCall!(
   VectorToRotate=Transform(None=ComponentMask(None=@n53) {b=true, g=true, r=true}) {transformSourceType=TRANSFORMSOURCE_World, transformType=TRANSFORM_Local}
   PreSkinnedLocalNormal=NamedRerouteUsage()
   VertexlNormalWS=Transform(None=VertexNormalWS()) {transformSourceType=TRANSFORMSOURCE_World, transformType=TRANSFORM_Local}
  ) {materialFunction=/Game/Packages/Fortress_SharedMaterials/Base_Material_Functions/RotateVector_FromSkinning}
  n54 = Custom!(SH=@MaterialExpressionStaticSwitchParameter_32, Direction=@MaterialExpressionMaterialFunctionCall_133.RotatedVectorFromSkinning) {
   code="//SH decode,  Direction = LightDir, SH = 2band coefficients. (encoded in Y-up)\r\nDirection = normalize(Direction);\r\nfloat4 shEvalDirection;\r\nshEvalDirection.x = 0.282095;\r\nshEvalDirection.y =0.488603 * Direction.z;\r\nshEvalDirection.z = 0.488603 * Direction.y;\r\nshEvalDirection.w =-0.488603 * Direction.x;\r\nreturn saturate(dot(SH, shEvalDirection));"
   description=SphericalHarmonicsEval
   outputType=CMOT_Float1
   inputs [
    {inputName=SH, input {expression=&self:MaterialExpressionStaticSwitchParameter_32}}
    {
     inputName=Direction
     input {expression=&self:MaterialExpressionMaterialFunctionCall_133}
    }
   ]
  }
  n55 = ScalarParameter() {group=Slurp, parameterName=SHContrast_LowValue}
  n56 = ScalarParameter() {defaultValue=1, group=Slurp, parameterName=SHContrast_HighValue}
  n57 = Constant()
  n58 = Constant() {r=1}
  n59 = Saturate(None=OneMinus(None=MaterialFunctionCall(Input=@n54, Input Low=@n55, Input High=@n56, Target Low=@n57, Target High=@n58) {materialFunction=&RemapValueRange}.Result))
  n60 = ScalarParameter() {defaultValue=1, group=Slurp, parameterName=SHPow}
  NamedRerouteDeclaration(None=VertexInterpolator(VS=Power(Base=@n59, Exp=@n60)).PS) {name=Sphereical Harmonics, nodeColor=<rgba 0 1 0.1328 1>}
  n61 = PixelNormalWS()
  n62 = CollectionParameter() {
   collection=&FortniteMaterialParameters
   parameterName=SunlightDirectionalVector
  }
  n63 = ComponentMask(None=@n62) {b=true, g=true, r=true}
  n64 = Multiply(A=@n61, B=Multiply(A=DotProduct(A=@n61, B=@n63)) {constB=2})
  n65 = DotProduct(
   A=Normalize(VectorInput=Subtract(A=@n64, B=@n63))
   B=Multiply(A=MaterialFunctionCall() {materialFunction=&CameraDirectionVector}.Result) {constB=-1}
  )
  n66 = Power(Base=Max(A=@n65) {constB=0}, Exp=ScalarParameter() {defaultValue=5, group=Slurp, parameterName=Slurp_InnerScatteringPower})
  n67 = ScalarParameter() {defaultValue=0.4, group=Slurp, parameterName=Slurp_InnerScatteringScale}
  NamedRerouteDeclaration(None=Saturate(None=Multiply(A=@n66, B=@n67) {constB=0.4})) {name=InnerScattering, nodeColor=<rgba 0.7578 1 0 1>}
  n68 = Multiply(A=@n65) {constB=-1}
  n69 = ScalarParameter() {group=Slurp, parameterName=Slurp_InnerBackScatteringBias}
  n70 = Power(
   Base=Max(A=Add(A=@n68, B=@n69)) {constB=0}
   Exp=ScalarParameter() {defaultValue=1, group=Slurp, parameterName=Slurp_InnerBackScatteringPower}
  )
  n71 = ScalarParameter() {group=Slurp, parameterName=Slurp_InnerBackScatteringScale}
  NamedRerouteDeclaration(None=Saturate(None=Multiply(A=@n70, B=@n71) {constB=0.55})) {name=InnerBackScattering, nodeColor=<rgba 0 0.6719 1 1>}
 ]}
