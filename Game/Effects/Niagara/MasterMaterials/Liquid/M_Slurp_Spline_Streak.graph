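# Translucent subsurface master material for Niagara mesh particles that draws liquid Slurp streaks along thickened splines, with a parameter-blended chrome variant.
# SplineThicken sets width and offset, and its sphere-mapped normal feeds MF_Slurp_forFX for shading.
# Panning sine wobble and flowmap distortion warp UVs that drive that normal and a sphere-mask opacity with erosion and camera fade.
# Chrome overrides surface values and adds swirl-normal vertex displacement, while dynamic parameters provide per-particle control.
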Material /Game/Effects/Niagara/MasterMaterials/Liquid/M_Slurp_Spline_Streak
graph {
 properties {
  bTangentSpaceNormal=false
  bUsedWithNiagaraMeshParticles=true
  blendMode=BLEND_Translucent
  pixelDepthOffsetMode=PDOM_Legacy
  shadingModel=MSM_Subsurface
  translucencyLightingMode=TLM_Surface
  usedShadingModels=Subsurface
 }
 nodes [
  n1 = DynamicParameter() {parameterIndex=1, paramNames [Pan,WidthMult,Param3,Param4]}
  n2 = Multiply(A=@n1.WidthMult, B=ScalarParameter() {defaultValue=25})
  n3 = DynamicParameter() {parameterIndex=1, paramNames [Pan,WidthMult,Param3,Param4]}
  n4 = ScalarParameter() {group=Chrome, parameterName=WidthBase, sortPriority=9}
  n5 = DynamicParameter() {parameterIndex=1, paramNames [Pan,WidthMult,Param3,Param4]}
  n6 = ScalarParameter() {group=Chrome, parameterName=WidthTip, sortPriority=9}
  n7 = MaterialFunctionCall(UV Channel=TextureCoordinate()) {materialFunction=&LinearGradient}
  n8 = NamedRerouteUsage()
  n9 = AppendVector(
   A=@n7.UGradient
   B=MaterialFunctionCall!(Center Point=ScalarParameter() {defaultValue=0.5, parameterName=StretchGradient Center Point}, 0-1 Gradient=@n7.VGradient, Scale Factor=@n8) {materialFunction=/Engine/Functions/Engine_MaterialFunctions02/Gradients/StretchGradient}.Resulting Gradient
  )
  n10 = AppendVector(A=@n7.UGradient, B=Multiply(A=@n7.VGradient, B=@n8))
  n11 = NamedRerouteUsage()
  n12 = OneMinus(None=@n7.VGradient)
  n13 = NamedRerouteUsage()
  n14 = Multiply(A=@n7.VGradient, B=Multiply(A=NamedRerouteUsage()) {constB=0.5})
  n15 = Multiply(A=Time(), B=ScalarParameter() {defaultValue=-0.2, parameterName=PanSpeed})
  n16 = Add(A=VertexColor().R, B=ParticleRandom())
  n17 = Fmod(A=@n16, B=Constant() {r=1})
  n18 = Add(A=Multiply(A=Add(A=@n15, B=@n1.Pan), B=LinearInterpolate(Alpha=@n17) {constA=1, constB=0.4}), B=@n16)
  n19 = ConstantBiasScale(None=Sine(None=Add(A=@n14, B=@n18)))
  n20 = ScalarParameter() {defaultValue=-0.2, parameterName=SineWobbleSizeSubtract}
  n21 = Add(
   A=MaterialFunctionCall(In=@n12, Power=LinearInterpolate(Alpha=@n13) {constA=0.5, constB=0}) {materialFunction=&SCurve}.Result
   B=Multiply(A=Multiply(A=@n19, B=@n20) {constB=0.1}, B=NamedRerouteUsage())
  )
  n22 = Constant() {r=1}
  n23 = NamedRerouteUsage()
  n24 = TextureCoordinate() {vTiling=0.5}
  n25 = AppendVector(A=Constant() {r=1}, B=NamedRerouteUsage())
  n26 = TextureSample(UVs=Panner(Coordinate=Multiply(A=@n24, B=@n25), Time=@n18) {speedX=0.1, speedY=-1}) {samplerType=SAMPLERTYPE_LinearColor, texture=/Game/Environments/Apollo/Sets/Jerky/Materials/T_JerkyFlowmap}
  n27 = Multiply(A=ConstantBiasScale(None=ComponentMask(None=@n26.RGB) {g=true, r=true}) {bias=-0.5, scale=2}) {constB=0.2}
  n28 = Add(
   A=MaterialFunctionCall(UVs=LinearInterpolate(A=@n9, B=@n10, Alpha=@n11), Texture Scale=AppendVector(A=@n21, B=@n22)) {materialFunction=&ScaleUVsByCenter}.UVs
   B=LinearInterpolate(A=Multiply(A=@n23, B=@n27), Alpha=@n12) {constB=0}
  )
  n29 = ScalarParameter() {defaultValue=5, parameterName=EndingNormalIntesnity}
  n30 = MaterialFunctionCall(R=@n29, G=@n29, B=Constant() {r=1}) {materialFunction=&AppendMany}
  n31 = DynamicParameter() {defaultValue=<rgba 0 0 7 0>, paramNames [Erode,Wobble,Tile,Separate]}
  n32 = Power(Base=NamedRerouteDeclaration(None=@n31.Erode) {name=ErodeParam, nodeColor=<rgba 0.7344 0 1 1>}) {constExponent=3}
  n33 = MaterialFunctionCall(
   WidthBase=LinearInterpolate(A=@n2, B=Multiply(A=@n3.WidthMult, B=@n4), Alpha=NamedRerouteUsage())
   WidthTip=LinearInterpolate(A=@n2, B=Multiply(A=@n5.WidthMult, B=@n6), Alpha=NamedRerouteUsage())
   AdditionalNormal=Multiply(
    A=TextureSample(UVs=@n28) {samplerType=SAMPLERTYPE_Normal, texture=&T_Sphere_N}.RGB
    B=LinearInterpolate(B=@n30.RGB, Alpha=@n32) {constA=1}
   )
  ) {materialFunction=&SplineThicken}
  n34 = GetMaterialAttributes(None=MaterialFunctionCall!(Pre Skinned Position, World Normal=Transform(None=@n33.Normal), AO) {materialFunction=&/MF_Slurp_forFX}.Result) {attributeGetTypes [
    #1
    #2
    #3
    #4
    #5
   ]}
  LinearInterpolate(
   A=@n34.Base Color
   B=VectorParameter() {group=Chrome, parameterName=ChromeBaseColor, sortPriority=1, defaultValue=<rgba 0 0 0 1>}.RGB
   Alpha=NamedRerouteUsage()
  )
  LinearInterpolate(
   A=@n34.Metallic
   B=ScalarParameter() {defaultValue=0.95, group=Chrome, parameterName=ChromeMetallic, sortPriority=2}
   Alpha=NamedRerouteUsage()
  )
  n35 = DynamicParameter() {parameterIndex=2, paramNames [PanAlongSpline, Pan Speed X, Pan Speed Y, WPO Mult]}
  n36 = Panner(Coordinate=TextureCoordinate(), Time=Time(), Speed=AppendVector(A=@n35.Pan Speed X, B=@n35.Pan Speed X))
  n37 = ComponentMask(None=@n36) {r=true}
  n38 = ScalarParameter() {defaultValue=1, group=Chrome, parameterName=ChromeXTile, sortPriority=6}
  n39 = ComponentMask(None=@n36) {g=true}
  n40 = ScalarParameter() {defaultValue=2.5, group=Chrome, parameterName=ChromeYTile, sortPriority=7}
  n41 = TextureSample(UVs=AppendVector(A=Multiply(A=@n37, B=@n38), B=Multiply(A=@n39, B=@n40))) {samplerType=SAMPLERTYPE_Normal, texture=/Game/Global/Textures/Default/Flowmaps/T_SwirlFlow_02_N}
  n42 = DynamicParameter() {parameterIndex=2, paramNames [PanAlongSpline, Pan Speed X, Pan Speed Y, WPO Mult]}
  LinearInterpolate(A=@n33.WorldPositionOffset, B=Add(A=@n33.WorldPositionOffset, B=Multiply(A=@n41.RGB, B=@n42.WPO Mult)), Alpha=NamedRerouteUsage())
  LinearInterpolate(
   A=@n34.Specular
   B=ScalarParameter() {defaultValue=0.95, group=Chrome, parameterName=ChromeSpecular, sortPriority=3}
   Alpha=NamedRerouteUsage()
  )
  LinearInterpolate(A=@n34.Roughness, B=ScalarParameter() {group=Chrome, parameterName=ChromeRoughness, sortPriority=4}, Alpha=NamedRerouteUsage())
  n43 = ScalarParameter() {defaultValue=0.5, parameterName=EmissiveFresnelExp}
  n44 = Power(Base=OneMinus(None=Fresnel(ExponentIn=@n43) {baseReflectFraction=0, exponent=7})) {constExponent=3}
  MaterialFunctionCall(Emissive=LinearInterpolate(
    A=Multiply(A=@n34.Emissive Color, B=@n44)
    B=ScalarParameter() {group=Chrome, parameterName=ChromeEmissive, sortPriority=5}
    Alpha=NamedRerouteUsage()
   )) {materialFunction=&MF_FX_EmissiveAdjustments}
  NamedRerouteDeclaration(None=@n31.Tile) {name=TileParam, nodeColor=<rgba 0.5469 0 1 1>}
  NamedRerouteDeclaration(None=@n31.Wobble) {name=WobbleParam, nodeColor=<rgba 0.7344 0 1 1>}
  NamedRerouteDeclaration(None=ScalarParameter() {group=Chrome, parameterName=UseChrome, sortPriority=0}) {name=UseChrome, nodeColor=<rgba 1 0.1 0 1>}
  NamedRerouteDeclaration(None=@n31.Separate) {name=SeparateParam, nodeColor=<rgba 0.9922 0 1 1>}
  n45 = MaterialFunctionCall(Float2=@n28) {materialFunction=&BreakOutFloat2Components}
  n46 = DynamicParameter() {parameterIndex=2, paramNames [PanAlongSpline, Param2, Param3, Param4]}
  n47 = Fmod(A=LinearInterpolate(A=@n45.G, B=Add(A=@n45.G, B=@n46.PanAlongSpline), Alpha=NamedRerouteUsage()), B=Constant() {r=2})
  n48 = Multiply(
   A=SphereMask(A=AppendVector(A=@n45.R, B=@n47), B=Constant() {r=0.5}, Radius=LinearInterpolate(Alpha=@n32) {constA=0.5, constB=0})
   B=MaterialFunctionCall(Distance=ScalarParameter() {defaultValue=250, parameterName=Near Camera Distance}) {materialFunction=&Near_Camera_Fade}.Result
  )
  n49 = ParticleColor()
  Reroute(None=StaticSwitchParameter(True=Multiply(A=@n48, B=@n49.A), False=@n48) {parameterName=Use Particle Alpha Control})
  Reroute(None=@n34.Base Color)
 ]
}
