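# Translucent unlit master material for Niagara water drip sprites, with responsive AA and index-of-refraction distortion.
# A drop normal map drives a fake reflection lookup, randomized per particle and lit by sunlight direction, then tinted by time of day for emissive.
# Drop texture, lighting and particle alpha form opacity, which distance-fades along with refraction.
# Quality and shading-path switches simplify lighting and normals, with Low also dropping fade and refraction.
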Material /Game/Effects/Niagara/MasterMaterials/Liquid/M_WaterDrip
graph {
 properties {
  bEnableResponsiveAA=true
  bUsedWithNiagaraSprites=true
  blendMode=BLEND_Translucent
  pixelDepthOffsetMode=PDOM_Legacy
  refractionMethod=RM_IndexOfRefraction
  shadingModel=MSM_Unlit
  usedShadingModels=Unlit
 }
 nodes [
  n1 = VectorParameter() {parameterName=Rain_Normal_Intensity, defaultValue=<rgba 0 0 1 1>}
  n2 = ShadingPathSwitch(
   Default=ShadingPathSwitch(
    Default=TextureSample() {samplerType=SAMPLERTYPE_Normal, texture=/Game/VisualThreatManager/StormVisuals/Textures/T_Water_drop_N_02}.RGB
    Mobile=@n1.RGB
   ) {desc=Migrated from FeatureLevelSwitch}
   Forward=@n1.RGB
  )
  LinearInterpolate(
   A=QualitySwitch(Default=@n2, Low=@n1.RGB, High=@n2, Medium=@n2)
   B=@n1.RGB
   Alpha=ScalarParameter() {parameterName=Remove_NMap_On_Switch_Mobile_Low?, sliderMax=1}
  )
  n3 = TextureSampleParameter2D() {parameterName=Normal, samplerType=SAMPLERTYPE_Normal, texture=&T_Water_drop_N_02}
  n4 = MaterialFunctionCall(
   Color_Morning=VectorParameter() {parameterName=Color Morning, defaultValue=<rgba 1 0.6506 0.3424 0>}.RGB
   Color_Day=VectorParameter() {parameterName=Color Day, defaultValue=<rgba 0.3004 0.4266 0.5 0>}.RGB
   Color_Evening=VectorParameter() {parameterName=Color Evening, defaultValue=<rgba 1.3 0.857 0.533 0>}.RGB
   Color_Night=VectorParameter() {parameterName=Color Night, defaultValue=<rgba 0.27 0.2723 1 0>}.RGB
  ) {materialFunction=&Color_Time_of_Day}
  n5 = TextureSampleParameter2D() {parameterName=Normal, samplerType=SAMPLERTYPE_Normal, texture=&T_Water_drop_N_02}
  n6 = MaterialFunctionCall!(Softness=ScalarParameter() {defaultValue=0.8, parameterName=Softness}, Normal=@n5.RGB) {materialFunction=/Engine/Functions/Engine_MaterialFunctions01/Reflections/ReflectionVectorSoft}
  n7 = ParticleRandom()
  n8 = TextureSample(UVs=ComponentMask(None=Add(A=@n6.Result, B=@n7)) {g=true, r=true}) {texture=/Game/Effects/Fort_Effects/Gameplay/Water/Textures/FX_Reflection_03_Diff}
  n9 = ScalarParameter() {defaultValue=0.01, parameterName=Bias}
  n10 = CollectionParameter() {
   collection=&FortniteMaterialParameters
   parameterName=SunlightDirectionalVector
  }
  n11 = DotProduct(
   A=@n3.RGB
   B=LinearInterpolate(
    A=VectorParameter() {parameterName=FakeSunlightVector, defaultValue=<rgba 0 0 1 0>}.RGB
    B=ComponentMask(None=@n10) {b=true, g=true, r=true}
    Alpha=ScalarParameter() {defaultValue=1, parameterName=UseSunlightDirectionalVector, sliderMax=1}
   )
  )
  n12 = Multiply(A=Multiply(A=Add(A=@n8.RGB, B=@n9), B=ScalarParameter() {defaultValue=4, parameterName=Brightness}), B=@n11)
  n13 = QualitySwitch(Default=@n12, Low=@n3.B)
  n14 = QualitySwitch(Default=ShadingPathSwitch(Default=@n3.B, Deferred=Multiply(A=@n4.Result, B=@n13)), Low=@n3.B)
  n15 = ParticleColor()
  MaterialFunctionCall(Emissive=LinearInterpolate(A=@n14, B=Multiply(A=@n14, B=@n15.RGB), Alpha=ScalarParameter() {parameterName=UsePaticleColor, sliderMax=1})) {materialFunction=&MF_FX_EmissiveAdjustments}
  n16 = TextureSample() {texture=/Game/VisualThreatManager/StormVisuals/Textures/T_Water_drop_02}
  n17 = QualitySwitch(Default=Add(A=@n12, B=Multiply(A=@n11, B=@n16.B)), Low=@n3.B)
  n18 = MaterialFunctionCall(
   Position=ParticlePositionWS()
   Fade_Start_Distance=ScalarParameter() {defaultValue=200, parameterName=Start Fade Distance}
   Fade_End_Distance=ScalarParameter() {defaultValue=150, parameterName=End Fade Distance}
  ) {materialFunction=&MF_Distance_SyncedFade_FX}
  Multiply(A=Multiply(A=@n15.A, B=Multiply(A=@n17, B=@n16.A)), B=QualitySwitch(Default=@n18.Result, Low=Constant() {r=1}))
  n19 = LinearInterpolate(B=LinearInterpolate(B=ScalarParameter() {defaultValue=0.9, parameterName=Refraction}, Alpha=@n16.A) {constA=1}, Alpha=ParticleColor().A) {constA=1}
  Reroute(None=QualitySwitch(Default=LinearInterpolate(B=@n19, Alpha=@n18.Result) {constA=1}, Low=Constant() {r=1}))
 ]
}
