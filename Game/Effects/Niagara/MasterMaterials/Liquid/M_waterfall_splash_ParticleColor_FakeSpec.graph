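# Unlit translucent Niagara master material for waterfall splash sprites and mesh particles.
# Fakes specular with a cubemap reflection and sun highlight off a distorted splash normal, blended with particle color or HSV-offset by it.
# Splash texture and panning noise form opacity, eroded by a dynamic parameter and softened by depth and camera fades.
# Optional WPO follows water displacement; Mobile and Low quality skip the fake lighting.
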Material /Game/Effects/Niagara/MasterMaterials/Liquid/M_waterfall_splash_ParticleColor_FakeSpec
graph {
 properties {
  bTangentSpaceNormal=false
  bUsedWithNiagaraMeshParticles=true
  bUsedWithNiagaraSprites=true
  blendMode=BLEND_Translucent
  pixelDepthOffsetMode=PDOM_Legacy
  shadingModel=MSM_Unlit
  translucencyLightingMode=TLM_VolumetricPerVertexNonDirectional
  translucencyPass=MTP_BeforeDOF
  translucentSelfShadowDensityScale=0
  translucentSelfShadowSecondDensityScale=0
  translucentShadowDensityScale=0
  usedShadingModels=Unlit
 }
 nodes [
  n1 = Panner(Coordinate=TextureCoordinate() {uTiling=2, vTiling=2}) {speedY=-0.06}
  n2 = TextureCoordinate()
  n3 = TextureSample(UVs=Panner(Coordinate=TextureCoordinate() {uTiling=2, vTiling=2}) {speedY=-0.03}) {texture=/Game/Global/Textures/Default/Tileable_Patterns/T_Tilling_Distortion_Map_01}
  n4 = ComponentMask(None=@n3.G) {g=true, r=true}
  n5 = ShadingPathSwitch(Default=Add(A=@n2, B=Multiply(A=@n4) {constB=0.1}), Mobile=@n2) {desc=Migrated from FeatureLevelSwitch}
  n6 = ReflectionVectorWS(CustomWorldNormal=Transform(None=TextureSample(UVs=@n5) {samplerType=SAMPLERTYPE_Normal, texture=/Game/Effects/Fort_Effects/Textures/Water/T_waterSplash_N}.RGB))
  n7 = TextureSampleParameterCube!(
   UVs=Add(
    A=TextureSample(UVs=@n1) {samplerType=SAMPLERTYPE_Normal, texture=/Game/Global/Textures/Default/Tileable_Patterns/T_Tilling_Distortion_Map_01_n}.RGB
    B=@n6
   )
   Apply View MipBias
  ) {parameterName=Cubemap, samplerType=SAMPLERTYPE_LinearColor, texture=/Game/Packages/Fortress_Sky/TexturesHDR/T_AthenaSkylight}
  n8 = ComponentMask(None=MaterialFunctionCall() {materialFunction=/Engine/Functions/GetPostProcessFunctions/GetAmbientCubemapTint}.AmbientCubemapTint) {b=true, g=true, r=true}
  n9 = CollectionParameter() {
   collection=&FortniteMaterialParameters
   parameterName=SunAndMoonModelDirectionalVector
  }
  n10 = ComponentMask(None=@n9) {b=true, g=true, r=true}
  n11 = Power(Base=Clamp(None=DotProduct(A=@n10, B=@n6))) {constExponent=20}
  n12 = Power(Base=ComponentMask(None=CollectionParameter() {collection=&FortniteMaterialParameters, parameterName=SunLightColor}) {b=true, g=true, r=true}) {constExponent=0.1}
  n13 = Add(
   A=Multiply(A=Multiply(A=@n7.RGB, B=@n8), B=VectorParameter() {parameterName=Cube Map Multiplier, defaultValue=<rgba 0.75 0.75 0.75 1>}.RGB)
   B=Multiply(A=Clamp(None=@n11) {maxDefault=5}, B=Clamp(None=@n12))
  )
  n14 = NamedRerouteUsage()
  n15 = Add(A=@n13, B=@n14)
  n16 = Multiply(A=@n13, B=@n14)
  n17 = ScalarParameter() {parameterName=MultiplyParticleColor}
  n18 = TextureSample(UVs=@n5) {samplerType=SAMPLERTYPE_LinearColor, texture=/Game/Effects/Fort_Effects/Textures/Water/T_waterSplash}
  n19 = OneMinus(None=@n18.G)
  n20 = Add(A=Multiply(A=Clamp(None=@n19)) {constB=0.125}) {constB=0.6}
  n21 = ShadingPathSwitch(Default=LinearInterpolate(A=LinearInterpolate(A=@n15, B=@n16, Alpha=@n17), Alpha=Clamp(None=@n20)) {constB=0}, Mobile=NamedRerouteUsage()) {desc=Migrated from FeatureLevelSwitch}
  n22 = Multiply(
   A=MaterialFunctionCall(
    Value_Morning=ScalarParameter() {defaultValue=1, parameterName="Value_Morning (S)"}
    Value_Day=ScalarParameter() {defaultValue=1, parameterName="Value_Day (S)"}
    Value_Evening=ScalarParameter() {defaultValue=1, parameterName="Value_Evening (S)"}
    Value_Night=ScalarParameter() {defaultValue=1, parameterName="Value_Night (S)"}
   ) {materialFunction=&Scalar_Time_of_Day}.Result
   B=NamedRerouteUsage()
  )
  n23 = Constant3Vector()
  n24 = ParticleColor()
  n25 = ScalarParameter() {parameterName=Use Particle Color RGB as HSV Offset}
  MaterialFunctionCall(Emissive=MaterialFunctionCall!(RGB=QualitySwitch(Default=@n21, Low=@n22), HSV Offset=LinearInterpolate(A=@n23, B=@n24.RGB, Alpha=@n25), Optional Reference Color) {materialFunction=/Game/Effects/Niagara/Emitters/Water/FoamEffect/MF_HSV_Offset}.Result) {materialFunction=&MF_FX_EmissiveAdjustments}
  n26 = DynamicParameter() {paramNames [Erode,Param2,Param3,Param4]}
  n27 = Multiply(
   A=Power(Base=Multiply(A=@n18.G, B=@n3.G), Exp=LinearInterpolate(Alpha=@n26.Erode) {constA=5, constAlpha=0}) {constExponent=50}
   B=LinearInterpolate(Alpha=@n26.Erode) {constAlpha=0, constB=200}
  )
  n28 = Multiply(A=ParticleColor().A, B=Multiply(A=@n18.G, B=Clamp(None=@n27)))
  n29 = ScalarParameter() {defaultValue=100, parameterName=Depth Fade}
  Multiply(
   A=ShadingPathSwitch(Default=DepthFade(Opacity=@n28, FadeDistance=@n29), Mobile=@n28) {desc=Migrated from FeatureLevelSwitch}
   B=MaterialFunctionCall(Distance=ScalarParameter() {parameterName=CameraFade}) {materialFunction=&Near_Camera_Fade}.Result
  )
  NamedRerouteDeclaration(None=LinearInterpolate(
    A=ParticleColor().RGB
    B=VectorParameter() {desc=fallback for RGB only, parameterName=ParticleColor RGB fallback, defaultValue=<rgba 1 1 1 1>}.RGB
    Alpha=ScalarParameter() {parameterName=Use Particle Color RGB as HSV Offset}
   )) {name=Particle Color RGB, nodeColor=<rgba 0 1 0.3672 1>}
  n30 = Multiply(
   A=MaterialFunctionCall!(Ignore XY Offsets=StaticBoolParameter() {defaultValue=true, parameterName=Ignore XY Gerstner Offsets}, Position, WaterBodyIndex, Depth) {materialFunction=/FortWater/Materials/Functions/WaterParticleDisplacement}.WPO
   B=VectorParameter() {parameterName=Water Displacement Multiplier, defaultValue=<rgba 1 1 1 0>}.RGB
  )
  n31 = MaterialFunctionCall(Z=ScalarParameter() {defaultValue=1, parameterName=ZOffset}) {materialFunction=&MakeFloat3}
  StaticSwitchParameter(
   True=LinearInterpolate(A=Constant(), B=Add(A=@n30, B=@n31.Result), Alpha=ScalarParameter() {defaultValue=1, parameterName=UseWaterDisplacement})
   False=Constant()
  ) {defaultValue=true, parameterName=UseWPO}
 ]
}
